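# Sound cue for the break-open step of the Monsoon shotgun reload, played as weapon foley with reload attenuation.
# A mixer layers a click, a break-open sound and a louder shell layer.
# The break-open sound is pitched up, with a volume envelope that ramps in quickly.
# One of three shell variants plays after a short random delay with modulation, and overridden concurrency blocks new instances at the limit.
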SoundCue /Game/Sounds/Fort_Weapons/Shotgun/Reload_Elements/Weapon_Monsoon_Break_Cue
root=Mixer_0
nodes [
 {id=Delay_0, children [Random_3], properties {delayMax=0.4, delayMin=0.2}}
 {
  id=Enveloper_0
  children [WavePlayer_3]
  properties {
   pitchCurve {editorCurveData {keys [{time=0}]}}
   volumeCurve {editorCurveData {keys [<at 0.0001 0.0282>,<at 0.1001 0.2025>,<at 0.1221 1.5907>]}}
  }
 }
 {id=Mixer_0, children [Modulator_0,Modulator_1,WavePlayer_0], properties {inputVolume [0.5,1,0.5]}}
 {id=Modulator_0, children [Enveloper_0], properties {pitchMax=1.5, pitchMin=1.5}}
 {id=Modulator_1, children [Delay_0]}
 {id=Random_3, children [WavePlayer_12, WavePlayer_13, WavePlayer_14], properties {weights [1,1,1]}}
 {id=WavePlayer_0, properties {soundWaveAssetPtr=&/weapon_shotgun_2_click}}
 {id=WavePlayer_12, properties {soundWaveAssetPtr=&/weapon_shotgun_break_shells_01}}
 {id=WavePlayer_13, properties {soundWaveAssetPtr=&/weapon_shotgun_break_shells_02}}
 {id=WavePlayer_14, properties {soundWaveAssetPtr=&/weapon_shotgun_break_shells_03}}
 {id=WavePlayer_3, properties {soundWaveAssetPtr=&/weapon_shotgun_break_open_01}}
]
properties {
 attenuationSettings=&Fort_Weapon_Reload_Att
 bOverrideConcurrency=true
 maxDistance=2250
 soundClassObject=&Fort_WeaponFoley
 volumeMultiplier=0.802
 concurrencyOverrides {resolutionRule=PreventNew}
}
