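# Sound cue for the second shell-insert element of a break-action shotgun reload.
# A Random node picks with equal weight between two shell-insert wave variations.
# Playback routes through the weapon foley sound class with weapon reload attenuation and a slight volume boost.
# A concurrency override uses the PreventNew rule, so new plays are rejected once the limit is reached.
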SoundCue /Game/Sounds/Fort_Weapons/Shotgun/Reload_Elements/Weapon_SG_Break_Reload_Insert2_Cue
root=Random_2
nodes [
 {id=Random_2, children [WavePlayer_0,WavePlayer_1], properties {weights [1,1]}}
 {id=WavePlayer_0, properties {soundWaveAssetPtr=&/weapon_shotgun_insert_shell_03}}
 {id=WavePlayer_1, properties {soundWaveAssetPtr=&/weapon_shotgun_insert_shell_04}}
]
properties {
 attenuationSettings=&Fort_Weapon_Reload_Att
 bOverrideConcurrency=true
 maxDistance=2250
 soundClassObject=&Fort_WeaponFoley
 volumeMultiplier=1.103
 concurrencyOverrides {resolutionRule=PreventNew}
}
