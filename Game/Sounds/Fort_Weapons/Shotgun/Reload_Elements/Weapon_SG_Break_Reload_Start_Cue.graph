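# Sound cue for the start of a break-action shotgun reload, layering the barrel-open sound with shell foley.
# A mixer blends a fixed, pitched-up break-open wave with a slightly quieter shell layer.
# The shell layer picks one of three variations at random, plays after a short random delay and gets a randomized, lowered pitch.
# The cue uses weapon reload attenuation, the weapon foley sound class and a concurrency override that prevents new instances at the limit.
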SoundCue /Game/Sounds/Fort_Weapons/Shotgun/Reload_Elements/Weapon_SG_Break_Reload_Start_Cue
root=Mixer_0
nodes [
 {id=Delay_0, children [Random_3], properties {delayMax=0.2, delayMin=0.1}}
 {id=Mixer_0, children [Modulator_0,Modulator_2], properties {inputVolume [1,0.8]}}
 {id=Modulator_0, children [WavePlayer_3], properties {pitchMax=1.2, pitchMin=1.2}}
 {id=Modulator_2, children [Delay_0], properties {pitchMax=0.95, pitchMin=0.85}}
 {id=Random_3, children [WavePlayer_12, WavePlayer_13, WavePlayer_14], properties {weights [1,1,1]}}
 {id=WavePlayer_12, properties {soundWaveAssetPtr=&/weapon_shotgun_break_shells_01}}
 {id=WavePlayer_13, properties {soundWaveAssetPtr=&/weapon_shotgun_break_shells_02}}
 {id=WavePlayer_14, properties {soundWaveAssetPtr=&/weapon_shotgun_break_shells_03}}
 {id=WavePlayer_3, properties {soundWaveAssetPtr=&/weapon_shotgun_break_open_01}}
]
properties {
 attenuationSettings=&Fort_Weapon_Reload_Att
 bOverrideConcurrency=true
 maxDistance=2250
 pitchMultiplier=0.9
 soundClassObject=&Fort_WeaponFoley
 volumeMultiplier=0.802
 concurrencyOverrides {resolutionRule=PreventNew}
}
